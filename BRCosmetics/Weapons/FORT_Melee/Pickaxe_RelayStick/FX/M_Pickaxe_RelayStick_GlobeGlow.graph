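# Translucent unlit material for the glow on the RelayStick pickaxe's globe, usable on Niagara mesh particles.
# Panning noise, mapped by camera-space normals, perturbs the world normal to make a wobbling Fresnel term.
# That term drives a soft rim opacity and a hard, brightened band between two step thresholds.
# A time-scrolled thin-film rainbow texture colors both glow and band, tinted by particle color through an emissive-adjustment function.
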Material /BRCosmetics/Weapons/FORT_Melee/Pickaxe_RelayStick/FX/M_Pickaxe_RelayStick_GlobeGlow
graph {
 properties {
  bUsedWithNiagaraMeshParticles=true
  blendMode=BLEND_Translucent
  pixelDepthOffsetMode=PDOM_Legacy
  shadingModel=MSM_Unlit
  usedShadingModels=Unlit
 }
 nodes [
  n1 = ParticleColor()
  n2 = TextureSample(UVs=Multiply(A=Time()) {constB=0.4}) {texture=&/T_Pickaxe_RelayStick_ThinFilmRainbow}
  n3 = VertexNormalWS()
  n4 = ComponentMask(None=Transform(None=VertexNormalWS()) {transformSourceType=TRANSFORMSOURCE_World, transformType=TRANSFORM_Camera}) {g=true, r=true}
  n5 = Constant2Vector() {g=0.5, r=-0.5}
  n6 = Multiply(A=Add(A=Multiply(A=@n4, B=@n5) {constB=2}) {constB=0.5}) {constB=0.6}
  n7 = Multiply(A=TextureSample(UVs=Panner(Coordinate=@n6) {speedX=0.5, speedY=1}) {samplerType=SAMPLERTYPE_Grayscale, texture=/FortUILibrary/UIKit/Buttons/Textures/LowResBlurredNoise}.R) {constB=0.15}
  n8 = OneMinus(None=DotProduct(A=Add(A=@n3, B=@n7), B=CameraVectorWS()))
  n9 = Multiply(A=Multiply(A=@n2.RGB) {constB=5}, B=Subtract(A=Step(X=@n8) {constY=0.6}, B=Step(X=@n8) {constX=0.7, constY=0.7}))
  MaterialFunctionCall(Emissive=Multiply(A=@n1.RGB, B=Add(A=@n2.RGB, B=@n9))) {materialFunction=&MF_FX_EmissiveAdjustments}
  Multiply(A=@n1.A, B=SmoothStep(Value=@n8) {constMax=0.8, constMin=0.25})
 ]
}
