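# Translucent unlit material for the RelayStick pickaxe's rainbow trail, usable on Niagara ribbons.
# A tapered UV mask combines with trail and panning smoke textures, while stepped, sine-wobbled mirrored UVs cut hard edge lines.
# Scrolling thin-film rainbow and curve-atlas gradient colors, tinted by particle color, feed an emissive adjustment function.
# RemoveHardLines blends opacity toward soft smoke before depth fade, and UseSolidColors blends color toward the tinted gradient alone.
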Material /BRCosmetics/Weapons/FORT_Melee/Pickaxe_RelayStick/FX/M_Pickaxe_RelayStick_RainbowTrail
graph {
 properties {
  bUsedWithInstancedStaticMeshes=true
  bUsedWithNanite=true
  bUsedWithNiagaraRibbons=true
  blendMode=BLEND_Translucent
  pixelDepthOffsetMode=PDOM_Legacy
  shadingModel=MSM_Unlit
  usedShadingModels=Unlit
 }
 nodes [
  n1 = TextureCoordinate()
  n2 = ComponentMask(None=@n1) {r=true}
  n3 = Abs(None=Subtract(A=ComponentMask(None=@n1) {g=true}) {constB=0.5})
  n4 = Subtract(A=Power(Base=OneMinus(None=@n2)) {constExponent=5}, B=Power(Base=Multiply(A=@n3) {constB=2}) {constExponent=3})
  n5 = TextureSample(UVs=Multiply(A=TextureCoordinate(), B=Constant2Vector() {g=1, r=4})) {samplerType=SAMPLERTYPE_LinearGrayscale, texture=&/T_Pickaxe_RelayStick_TrailTexture}
  n6 = Add(
   A=@n5.R
   B=TextureSample(UVs=Panner() {speedX=1.2}) {samplerType=SAMPLERTYPE_Grayscale, texture=&T_Smoke_Trail_Soft}.R
  )
  n7 = Add(A=Multiply(A=@n4, B=@n6), B=Multiply(A=@n4) {constB=0.2})
  n8 = Saturate(None=Power(Base=@n7))
  Add(A=@n8, B=@n5.R)
  n9 = ParticleColor()
  n10 = TextureCoordinate()
  n11 = Saturate(None=Subtract(A=ComponentMask(None=@n10) {r=true}) {constB=0.2})
  n12 = LinearInterpolate(Alpha=@n11) {constA=0.65, constB=0}
  n13 = NamedRerouteUsage()
  n14 = Step(Y=@n12, X=@n13) {constY=0.65}
  n15 = Subtract(A=@n14, B=Step(Y=Add(A=@n12) {constB=0.16}, X=@n13) {constY=0.77})
  n16 = OneMinus(None=@n14)
  n17 = Saturate(None=Multiply(A=@n7) {constB=3.5})
  n18 = ComponentMask(None=TextureCoordinate()) {r=true}
  n19 = SmoothStep(Value=OneMinus(None=@n18)) {constMax=0.6, constMin=0.2}
  n20 = Add(A=@n17, B=Multiply(A=@n6, B=@n19))
  n21 = Multiply(A=@n16, B=@n20)
  n22 = Multiply(A=Saturate(None=Add(A=@n15, B=@n21)), B=Multiply(A=SmoothStep(Value=@n18) {constMax=0.2, constMin=0.006}, B=@n19))
  n23 = ScalarParameter() {parameterName=RemoveHardLines}
  DepthFade(Opacity=Multiply(A=@n9.A, B=LinearInterpolate(A=@n22, B=@n20, Alpha=@n23))) {fadeDistanceDefault=12}
  n24 = LinearInterpolate(Alpha=@n11) {constA=0.725, constB=0}
  n25 = NamedRerouteUsage()
  n26 = Add(A=@n24) {constB=0.04}
  n27 = Saturate(None=Subtract(A=Step(Y=@n24, X=@n25) {constY=0.65}, B=Step(Y=@n26, X=@n25) {constY=0.77}))
  n28 = Multiply(A=ComponentMask(None=TextureCoordinate()) {r=true}) {constB=0.6}
  n29 = Multiply(A=Time()) {constB=0.4}
  n30 = Multiply(
   A=TextureSample(UVs=Add(A=@n28, B=@n29)) {texture=&/T_Pickaxe_RelayStick_ThinFilmRainbow}.RGB
   B=@n9.RGB
  )
  n31 = Add(A=Multiply(A=Add(A=@n8, B=@n27), B=Multiply(A=@n30) {constB=5}), B=@n30)
  n32 = Subtract(A=@n16, B=@n17)
  n33 = CurveAtlasRowParameter(CurveTime=OneMinus(None=@n6)) {
   atlas=/Game/Weapons/FORT_Melee/Pickaxe_RelayStick/Textures/CA_RelayStick
   curve=/Game/Weapons/FORT_Melee/Pickaxe_RelayStick/Textures/C_RelayStick_EnvMapTint
   parameterName=TintGradient
  }
  MaterialFunctionCall(Emissive=LinearInterpolate(A=LinearInterpolate(A=@n31, B=@n33, Alpha=@n32), B=Multiply(A=@n9.RGB, B=@n33), Alpha=ScalarParameter() {parameterName=UseSolidColors})) {materialFunction=&MF_FX_EmissiveAdjustments}
  n34 = TextureCoordinate() {coordinateIndex=1}
  n35 = Multiply(A=Abs(None=Subtract(A=@n34) {constB=0.5})) {constB=2}
  n36 = Time()
  n37 = Constant2Vector() {r=0.5}
  n38 = Multiply(A=Add(A=TextureCoordinate() {coordinateIndex=1}, B=Multiply(A=@n36, B=@n37))) {constB=6}
  n39 = Multiply(A=ComponentMask(None=Sine(None=@n38)) {r=true}) {constB=0.15}
  NamedRerouteDeclaration(None=ComponentMask(None=Add(A=@n35, B=@n39)) {g=true}) {name=MirroredLines, nodeColor=<rgba 0.9688 1 0 1>}
 ]
}
