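# Translucent unlit, two-sided Niagara sprite material for the RelayStick pickaxe's sound wave effect, drawing radial waves and a growing ring.
# Noise-distorted polar UVs scroll a trail texture under a radial falloff with a center glow, using explicit derivatives from unwrapped UVs for mip selection.
# Dynamic parameters drive the ring's growth, color boost and opacity.
# Noise-indexed spectrum and time-cycled rainbow textures supply color, tinted by particle color, with depth-faded opacity.
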Material /BRCosmetics/Weapons/FORT_Melee/Pickaxe_RelayStick/FX/M_Pickaxe_RelayStick_SoundWaves
graph {
 properties {
  bUsedWithNiagaraSprites=true
  blendMode=BLEND_Translucent
  pixelDepthOffsetMode=PDOM_Legacy
  shadingModel=MSM_Unlit
  twoSided=true
  usedShadingModels=Unlit
 }
 nodes [
  n1 = MaterialFunctionCall(XY=TextureCoordinate()) {materialFunction=&Swizzle}
  n2 = Distance(A=@n1.YX, B=Constant() {r=0.5})
  n3 = OneMinus(None=@n2)
  n4 = Multiply(A=Time(), B=Constant2Vector() {r=-0.3})
  n5 = Subtract(A=@n1.YX) {constB=0.5}
  n6 = Divide(A=Arctangent2Fast(Y=ComponentMask(None=@n5) {r=true}, X=ComponentMask(None=@n5) {g=true}), B=Constant() {r=6.28})
  n7 = Panner(Coordinate=Multiply(A=TextureCoordinate()) {constB=0.5}) {speedX=0.15, speedY=0.1}
  n8 = Add(
   A=AppendVector(A=@n2, B=Frac(None=@n6))
   B=Multiply(A=TextureSample(UVs=@n7) {samplerType=SAMPLERTYPE_Grayscale, texture=/FortUILibrary/UIKit/Buttons/Textures/LowResBlurredNoise}.R) {constB=0.1}
  )
  n9 = Constant2Vector() {g=2, r=0.5}
  n10 = TextureSample!(UVs=Add(A=@n4, B=Multiply(A=@n8, B=@n9)), Tex, "DDX(UVs)"=DDX(Value=@n1.YX), "DDY(UVs)"=DDY(Value=@n1.YX), Apply View MipBias) {
   mipValueMode=TMVM_Derivative
   samplerType=SAMPLERTYPE_LinearGrayscale
   texture=&/T_Pickaxe_RelayStick_TrailTexture
  }
  n11 = OneMinus(None=@n2)
  n12 = Add(A=Multiply(A=Power(Base=@n3) {constExponent=8}) {constB=0.3}, B=Multiply(A=@n10.R, B=SmoothStep(Value=@n11) {constMax=0.9, constMin=0.5}))
  n13 = DynamicParameter() {paramNames [CenterRingGrow, CenterRingColorBoost, CenterRingOpacity, Param4]}
  n14 = Multiply(A=NamedRerouteUsage(), B=@n13.CenterRingOpacity)
  n15 = ParticleColor()
  DepthFade(Opacity=Multiply(A=Add(A=@n12, B=@n14), B=@n15.A)) {fadeDistanceDefault=6}
  n16 = Multiply(A=TextureCoordinate()) {constB=0.2}
  n17 = TextureSampleParameter2D(UVs=TextureSample!(UVs=Panner(Coordinate=@n16) {speedX=0.5, speedY=1}, Tex, Apply View MipBias) {samplerType=SAMPLERTYPE_Grayscale, texture=&LowResBlurredNoise}.R) {parameterName=Inner_ColorTexture, texture=/Game/Global/Textures/Default/Thinfilm/T_ThinFilm_Spectrum_COLOR}
  n18 = TextureSampleParameter2D(UVs=Multiply(A=Time()) {constB=0.4}) {parameterName=GlowRing_ColorTexture, texture=&/T_Pickaxe_RelayStick_ThinFilmRainbow}
  n19 = Multiply(A=@n18.RGB, B=@n13.CenterRingColorBoost)
  n20 = Add(A=@n8, B=OneMinus(None=DynamicParameter() {paramNames [CenterRingGrow, CenterRingColorBoost, CenterRingOpacity, Param4]}.CenterRingGrow)) {constB=0}
  n21 = Abs(None=Subtract(A=ComponentMask(None=@n20) {r=true}) {constB=0.5})
  n22 = SmoothStep(Value=@n21) {constMax=0.01, constMin=0.04}
  MaterialFunctionCall(Emissive=Multiply(A=LinearInterpolate(A=@n17.RGB, B=@n19, Alpha=@n22), B=@n15.RGB)) {materialFunction=&MF_FX_EmissiveAdjustments}
  NamedRerouteDeclaration(None=Add(A=Multiply(A=@n22) {constB=0.25}, B=Step(Y=@n21) {constX=0.004, constY=0.2})) {name=GrowingRing, nodeColor=<rgba 0 1 0.9063 1>}
  SmoothStep(Value=Frac(None=@n4)) {constMax=0.7, constMin=0.3}
 ]
}
