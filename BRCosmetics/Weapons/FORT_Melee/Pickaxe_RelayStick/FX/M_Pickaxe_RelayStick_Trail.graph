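# Unlit translucent Niagara ribbon material for the swing trail of the RelayStick pickaxe.
# A figure texture on UV1 supplies color, blended with a curve-atlas tint gradient, multiplied by particle color and passed through emissive adjustments.
# Opacity combines the inverted texture red channel, a falloff along U, particle alpha and depth fade.
# A stepped V mask, wobbled by a time-scrolled sine and varied along U, cuts hard, wavy ribbon edges.
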Material /BRCosmetics/Weapons/FORT_Melee/Pickaxe_RelayStick/FX/M_Pickaxe_RelayStick_Trail
graph {
 properties {bUsedWithNiagaraRibbons=true, blendMode=BLEND_Translucent, pixelDepthOffsetMode=PDOM_Legacy, shadingModel=MSM_Unlit, usedShadingModels=Unlit}
 nodes [
  n1 = ParticleColor()
  n2 = ComponentMask(None=TextureCoordinate()) {r=true}
  n3 = TextureSample(UVs=TextureCoordinate() {coordinateIndex=1}) {texture=&/T_Pickaxe_RelayStick_TrailFigure}
  n4 = OneMinus(None=@n3.R)
  n5 = Multiply(A=@n4, B=Power(Base=OneMinus(None=@n2)) {constExponent=2.5})
  n6 = VertexInterpolator(VS=TextureCoordinate())
  n7 = Saturate(None=Subtract(A=ComponentMask(None=@n6.PS) {r=true}) {constB=0.2})
  n8 = Subtract(A=TextureCoordinate()) {constB=0.5}
  n9 = TextureCoordinate()
  n10 = Multiply(A=Time(), B=Constant2Vector() {r=0.5})
  n11 = Sine(None=Multiply(A=Add(A=@n9, B=@n10)) {constB=6})
  n12 = Add(A=Multiply(A=Abs(None=@n8)) {constB=2}, B=Multiply(A=ComponentMask(None=@n11) {r=true}) {constB=0.15})
  n13 = OneMinus(None=Step(Y=LinearInterpolate(Alpha=@n7) {constA=0.65, constB=0}, X=ComponentMask(None=@n12) {g=true}) {constY=0.65})
  DepthFade(Opacity=Multiply(A=@n1.A, B=Multiply(A=@n5, B=@n13))) {fadeDistanceDefault=5}
  n14 = CurveAtlasRowParameter(CurveTime=Power(Base=@n4) {constExponent=5}) {
   atlas=/Game/Weapons/FORT_Melee/Pickaxe_RelayStick/Textures/CA_RelayStick
   curve=/Game/Weapons/FORT_Melee/Pickaxe_RelayStick/Textures/C_RelayStick_EnvMapTint
   parameterName=TintGradient
  }
  n15 = Multiply(A=@n14) {constB=0.5}
  MaterialFunctionCall(Emissive=Multiply(A=LinearInterpolate(A=@n3.RGB, B=@n15) {constAlpha=0.25}, B=@n1.RGB)) {materialFunction=&MF_FX_EmissiveAdjustments}
 ]
}
