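# Idle Niagara effect for the RelayStick pickaxe: colour-cycling rings, a skeletal-mesh-attached orb glow with light, and a warm staff glow.
# MainRing feeds pulse values to its material and applies time-of-day tinting; OLDMainRing is disabled.
# System bIsReactive switches ring pulse and scale, and orb colour, between looping animation and Timeline/AmplitudeValue-driven response.
# Pulsing bursts short-lived meshes at the skeletal mesh, gated by an AmplitudeValue threshold.
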NiagaraSystem /BRCosmetics/Weapons/FORT_Melee/Pickaxe_RelayStick/FX/NS_Pickaxe_RelayStick_NonReactive_Idle
dependencies {
 usedDataInterfaces [&NiagaraDataInterfaceColorCurve, &NiagaraDataInterfaceCurve, &NiagaraDataInterfaceArrayFloat]
 usedDynamicInputs [&Power, &ClampFloat]
 usedModules [
  &self:SystemSpawnScript.SystemScriptSource.SystemScriptGraph.NiagaraNodeAssignment_1.SetVariables_#1
  &SystemState
  &self:SystemSpawnScript.SystemScriptSource.SystemScriptGraph.NiagaraNodeAssignment_2.SetVariables_#2
  &Timeline
 ]
}
emitters {
 MainRing {
  emitterUpdateScript [
   &EmitterState(Life Cycle Mode=::NewEnumerator1, Loop Behavior=::NewEnumerator1, Loop Duration=1, UseLoopDelay=false)
   &SpawnBurst_Instantaneous(Spawn Count=2, Use Spawn Probability=false, UseLoopCountLimit=false)
  ]
  particleSpawnScript [
   &FN_SetSpawnGroup(Number of Spawn Groups=2)
   &InitializeParticle(
    Color=<rgba 1 1 1 0.6>
    Lifetime=5
    Mesh Scale Mode=::NewEnumerator3
    Mesh Uniform Scale=&SelectFloatFromArray(Direct Array Index=$Particles.SpawnGroup, Float Selection Array=DataInterface() {propertyValues {FloatData [0.2,0.1]}})
    Position Offset=<0 7 49>
    UsePositionOffset=true
    Write Material Random=true
   )
  ]
  particleUpdateScript [
   &ParticleState(Kill Particles When Lifetime Has Elapsed=false)
   /SetVariables(
    Particles.BrightRingColor:LinearColor=&ColorFromCurve(
     ColorCurve=DataInterface() {propertyValues {
       AlphaCurve {keys [<at 0.4953 1 RCIM_Cubic>]}
       BlueCurve {keys [<at 0 0 RCIM_Cubic>,<at 0.1979 0>,<at 0.5342 1.7266>,<at 0.7607 0.7225>,<at 0.978 0>]}
       GreenCurve {keys [<at 0 0.4793 RCIM_Cubic>,<at 0.1979 2>,<at 0.5342 2>,<at 0.7607 0.1718>,<at 0.978 0.4793>]}
       RedCurve {keys [<at 0 1 RCIM_Cubic>,<at 0.1979 0.3323>,<at 0.5342 0>,<at 0.7607 1>,<at 0.978 1>]}
      }}
     CurveIndex=&Lerp_Float(A=$System.NormalizedLoopAge, B=$System.AmplitudeValue, Alpha=$System.Timeline.Time)
    )
    Particles.DarkRingColor:LinearColor=&ColorFromCurve(
     ColorCurve=DataInterface() {propertyValues {
       AlphaCurve {keys [<at 0.4953 0.3 RCIM_Cubic>]}
       BlueCurve {keys [
         <at 0 0>
         {arriveTangent=1.8175, interpMode=RCIM_Cubic, leaveTangent=1.8175, time=0.2116, value=0.2049}
         <at 0.5502 1>
         <at 0.7836 1.7266>
         <at 1 0>
        ]}
       GreenCurve {keys [
         <at 0 1>
         {arriveTangent=-1.569, interpMode=RCIM_Cubic, leaveTangent=-1.569, time=0.2116, value=0.0852}
         <at 0.5502 0.1367>
         <at 0.7836 2>
         <at 1 1>
        ]}
       RedCurve {keys [
         <at 0 0.3325>
         {arriveTangent=-0.0876, interpMode=RCIM_Cubic, leaveTangent=-0.0876, time=0.2116, value=1}
         <at 0.5502 0.2843>
         <at 0.7836 0>
         <at 1 0.3325>
        ]}
      }}
     CurveIndex=&Lerp_Float(A=$System.NormalizedLoopAge, B=$System.AmplitudeValue, Alpha=$System.Timeline.Time)
    )
   ) {isSetParameters=true, module=SetVariables_#3}
   &DynamicMaterialParameters(
    Index 0 Param 1=&MakeFloatFromVector(Vector=$Particles.Scale, Channel=::NewEnumerator0)
    Index 0 Param 2=&MakeCustomFloatFromBool(Bool=$System.bIsReactive, True Float=35, False Float=20)
    Index 0 Param 3=&MakeCustomFloatFromBool(
     Bool=$System.bIsReactive
     True Float=$System.ReactiveValue
     False Float=&FloatFromCurve(
      FloatCurve=DataInterface() {propertyValues {Curve {keys [
          {interpMode=RCIM_Cubic, leaveTangent=4.8905, tangentMode=RCTM_Break, time=0, value=0.15}
          <at 0.2 0.9>
          {interpMode=RCIM_Cubic, leaveTangent=-1.0585, tangentMode=RCTM_Break, time=0.3, value=0.9}
          {arriveTangent=-0.6609, tangentMode=RCTM_User, time=1, value=0.15}
         ]}}}
      CurveIndex=$System.NormalizedLoopAge
     )
    )
    Index 0 Param 4=&MakeCustomFloatFromBool(Bool=$System.bIsReactive, True Float=1, False Float=0)
    Module.Index0Param0Write=true
    Module.Index0Param1Write=true
    Module.Index0Param2Write=true
    Module.Index0Param3Write=true
   )
   &ScaleMeshSize(Scale Factor=&Lerp_Vector(
     A=&VectorFromFloat(Value=&Waveform(
       Global Input Position=&SelectFloatFromArray(Direct Array Index=&SpawnGroup(), Float Selection Array=DataInterface() {propertyValues {FloatData [1,1.2]}})
       "Waveform [1]"=::NewEnumerator0
       "[1] Phase Scale"=0.6
       "[1] Frequency"=1
       "[1] Amplitude Min/Max"=<0.8 1>
      ))
     B=&VectorFromFloat(Value=&Lerp_Float(A=1, B=1.25, Alpha=&Lerp_Float(A=0, B=$System.AmplitudeValue, Alpha=$System.Timeline.Time)))
     Alpha=&MakeCustomFloatFromBool(Bool=$System.bIsReactive, True Float=1, False Float=0)
    )) {module=ScaleMeshSize001}
   &ScaleColor(
    Scale Mode=::NewEnumerator1
    Scale RGBA=&MakeVector4_FromLinearColor(LinearColor=&FNTimeOfDayColor(Night Color=<rgba 0.4065 0.3505 0.7812 1>, Morning Color=<rgba 1 0.6536 0.6237 1>, Day Color=<rgba 1 1 1 1>, Evening Color=<rgba 1 1 1 1>))
    ScaleRGBA=true
   )
  ]
  rendererClasses [&NiagaraMeshRendererProperties]
  renderers [{rendererClass=&NiagaraMeshRendererProperties, rendererIndex=0}]
 }
 OLDMainRing {
  bEnabled=false
  emitterUpdateScript [
   &EmitterState(Life Cycle Mode=::NewEnumerator1, Loop Behavior=::NewEnumerator1, Loop Duration=1, UseLoopDelay=false)
   &SpawnBurst_Instantaneous(Spawn Count=2, Use Spawn Probability=false, UseLoopCountLimit=false)
  ]
  particleSpawnScript [
   &FN_SetSpawnGroup(Number of Spawn Groups=2)
   &InitializeParticle(
    Color=<rgba 3 2.2806 0.5447 1>
    Lifetime=5
    Mesh Scale Mode=::NewEnumerator3
    Mesh Uniform Scale=&SelectFloatFromArray(Direct Array Index=$Particles.SpawnGroup, Float Selection Array=DataInterface() {propertyValues {FloatData [0.165,0.1]}})
    Position Offset=<0 6 49>
    UsePositionOffset=true
    Write Material Random=true
   )
  ]
  particleUpdateScript [
   &ParticleState(Kill Particles When Lifetime Has Elapsed=false)
   /SetVariables(
    Particles.BrightRingColor:LinearColor=&ColorFromCurve(
     ColorCurve=DataInterface() {propertyValues {
       AlphaCurve {keys [<at 0.4953 1 RCIM_Cubic>]}
       BlueCurve {keys [<at 0 0 RCIM_Cubic>,<at 0.1979 0>,<at 0.5342 1.7266>,<at 0.7607 0.7225>,<at 0.978 0>]}
       GreenCurve {keys [<at 0 0.4793 RCIM_Cubic>,<at 0.1979 2>,<at 0.5342 2>,<at 0.7607 0.1718>,<at 0.978 0.4793>]}
       RedCurve {keys [<at 0 1 RCIM_Cubic>,<at 0.1979 0.3323>,<at 0.5342 0>,<at 0.7607 1>,<at 0.978 1>]}
      }}
     CurveIndex=&Lerp_Float(A=$System.NormalizedLoopAge, B=$System.AmplitudeValue, Alpha=$System.Timeline.Time)
    )
    Particles.DarkRingColor:LinearColor=&ColorFromCurve(
     ColorCurve=DataInterface() {propertyValues {
       AlphaCurve {keys [<at 0.4953 0.3 RCIM_Cubic>]}
       BlueCurve {keys [
         <at 0 0>
         {arriveTangent=1.8175, interpMode=RCIM_Cubic, leaveTangent=1.8175, time=0.2116, value=0.2049}
         <at 0.5502 1>
         <at 0.7836 1.7266>
         <at 1 0>
        ]}
       GreenCurve {keys [
         <at 0 1>
         {arriveTangent=-1.569, interpMode=RCIM_Cubic, leaveTangent=-1.569, time=0.2116, value=0.0852}
         <at 0.5502 0.1367>
         <at 0.7836 2>
         <at 1 1>
        ]}
       RedCurve {keys [
         <at 0 0.3325>
         {arriveTangent=-0.0876, interpMode=RCIM_Cubic, leaveTangent=-0.0876, time=0.2116, value=1}
         <at 0.5502 0.2843>
         <at 0.7836 0>
         <at 1 0.3325>
        ]}
      }}
     CurveIndex=&Lerp_Float(A=$System.NormalizedLoopAge, B=$System.AmplitudeValue, Alpha=$System.Timeline.Time)
    )
   ) {isSetParameters=true, module=SetVariables_#3}
   &DynamicMaterialParameters(
    Index 0 Param 1=&MakeFloatFromVector(Vector=$Particles.Scale, Channel=::NewEnumerator0)
    Index 0 Param 2=1
    Index 1 Param 1=5
    Index 1 Param 2=5
    Index 1 Param 4=&Multiply_Float(A=0.4, B=$System.AmplitudeValue)
    Module.Index0Param0Write=true
    Module.Index0Param1Write=true
    Module.Index0Param2Write=true
    Module.Index0Param3Write=true
    Module.Index1Param0Write=true
    Module.Index1Param1Write=true
    Module.Index1Param2Write=true
    Module.Index1Param3Write=true
    Module.Index2Param0Write=true
    Module.Index2Param1Write=true
    Module.Index2Param2Write=true
    Module.Index2Param3Write=true
    Write Parameter Index 1=true
    Write Parameter Index 2=true
   )
   &ScaleMeshSize(Scale Factor=&Lerp_Vector(
     A=&VectorFromFloat(Value=&Waveform(
       Global Input Position=&SelectFloatFromArray(Direct Array Index=&SpawnGroup(), Float Selection Array=DataInterface() {propertyValues {FloatData [1.2,2]}})
       "Waveform [1]"=::NewEnumerator0
       "[1] Phase Scale"=0.6
       "[1] Frequency"=1.5
       "[1] Amplitude Min/Max"=<0.85 1>
      ))
     B=&VectorFromFloat(Value=&Lerp_Float(A=1, B=1.25, Alpha=&Lerp_Float(A=0, B=$System.AmplitudeValue, Alpha=$System.Timeline.Time)))
     Alpha=&MakeCustomFloatFromBool(Bool=$System.bIsReactive, True Float=1, False Float=0)
    )) {module=ScaleMeshSize001}
  ]
  rendererClasses [&NiagaraMeshRendererProperties]
  renderers [
   {rendererClass=&NiagaraMeshRendererProperties, rendererIndex=0}
   {rendererClass=&NiagaraMeshRendererProperties, rendererIndex=1}
   {rendererClass=&NiagaraMeshRendererProperties, rendererIndex=2}
  ]
 }
 OrbGlow {
  emitterUpdateScript [
   &EmitterState(Life Cycle Mode=::NewEnumerator1, Loop Behavior=::NewEnumerator1, Loop Duration=1.5, UseLoopDelay=false)
   &SpawnBurst_Instantaneous(Use Spawn Probability=false, UseLoopCountLimit=false)
  ]
  particleSpawnScript [&InitializeParticle(
    Color=<rgba 1 1 1 0.7>
    Lifetime=5
    Sprite Size Mode=::NewEnumerator3
    Uniform Sprite Size=65
    UsePositionOffset=false
    Write Material Random=true
   )]
  particleUpdateScript [
   &ParticleState(Kill Particles When Lifetime Has Elapsed=false)
   &SkeletalMeshLocation(Bone Sampling Mode=::NewEnumerator4, MaskSpawnGroup=false, Skeletal Mesh=$User.SkeletalMesh)
   &ScaleColor(
    Curve Index=&MakeCustomFloatFromBool(
     Bool=$System.bIsReactive
     True Float=&Lerp_Float(A=$System.NormalizedLoopAge, B=$System.AmplitudeValue, Alpha=$System.Timeline.Time)
     False Float=$System.NormalizedLoopAge
    )
    Linear Color Curve=DataInterface() {propertyValues {
      AlphaCurve {keys [<at 0 1>,<at 1 1>]}
      BlueCurve {keys [<at 0 0>,<at 0.149 0.0823>,<at 0.3697 0.7991>,<at 0.5551 0.9131>,<at 0.7851 0.2462>,<at 1 0>]}
      GreenCurve {keys [<at 0 0.552>,<at 0.149 0.1248>,<at 0.3697 0.1441>,<at 0.5551 0.3916>,<at 0.7851 0.8308>,<at 1 0.552>]}
      RedCurve {keys [<at 0 1>,<at 0.149 1>,<at 0.3697 0.9823>,<at 0.5551 0.3325>,<at 0.7851 0.5333>,<at 1 1>]}
     }}
    Scale Mode=::NewEnumerator2
   )
   /SetVariables(Particles.SpriteSize=&Multiply_Vector2D_ByFloat(
     Float=&FloatFromCurve(FloatCurve=DataInterface() {propertyValues {Curve {keys [<at 0 1>,<at 0.5 1.15>]}}}, CurveIndex=$System.Timeline.Time)
     Vector2D=$Particles.Initial.SpriteSize
    )) {isSetParameters=true}
   &FN_Light_Attributes(
    Color=$Particles.Color
    Create Color Attribute=true
    Create Exponent Attribute=true
    Create Radius Attribute=true
    Enable Particle Light Enable=true
    Exponent=2
    Light Enabled=true
    Radius=&Lerp_Float(A=80, B=90, Alpha=$System.AmplitudeValue)
    Scale Color=3
   )
  ]
  rendererClasses [&NiagaraSpriteRendererProperties, &NiagaraLightRendererProperties]
  renderers [
   {rendererClass=&NiagaraSpriteRendererProperties}
   {rendererClass=&NiagaraLightRendererProperties, rendererIndex=1}
  ]
 }
 Pulsing {
  emitterUpdateScript [
   &EmitterState(Life Cycle Mode=::NewEnumerator1, Loop Duration=0.12, UseLoopDelay=false)
   &SpawnBurst_Instantaneous(
    Spawn Count=&MakeCustomIntFromBool(
     Bool=&SetBoolByFloatComparison(A=$System.AmplitudeValue, B=1, Comparison Type=::NewEnumerator4)
     True INT=1
     False INT=0
    )
    Use Spawn Probability=false
    UseLoopCountLimit=false
   )
  ]
  particleSpawnScript [
   &InitializeParticle(
    Color=&ColorFromCurve(
     ColorCurve=DataInterface() {propertyValues {
       AlphaCurve {keys [<at 0.4851 0.4>]}
       BlueCurve {keys [<at 0 0>,<at 0.2429 0.6734>,<at 0.5227 1>,<at 0.7851 0>,<at 1 0>]}
       GreenCurve {keys [<at 0 0.0452>,<at 0.2429 0.1165>,<at 0.5227 0.7791>,<at 0.7851 0.5973>,<at 1 0.0452>]}
       RedCurve {keys [<at 0 1>,<at 0.2429 1>,<at 0.5227 0.0248>,<at 0.7851 1>,<at 1 1>]}
      }}
     CurveIndex=$Emitter.NormalizedLoopAge
    )
    Lifetime Max=1
    Lifetime Min=0.75
    Lifetime Mode=::NewEnumerator1
    Mesh Scale Mode=::NewEnumerator3
    Mesh Uniform Scale=0.6
    Sprite Rotation Mode=::NewEnumerator2
    Sprite Size Mode=::NewEnumerator3
    Uniform Sprite Size=70
    UsePositionOffset=false
    Write Material Random=true
   )
   &InitialMeshOrientation(
    Orientation Axis=<0 0 0>
    Orientation Vector=<0 0 0>
    Rotation=&RandomRangeVector(Minimum=<0 0 0>, Maximum=<1 0 0>)
    Use Orientation Vector=false
    Use Rotation Vector=true
   ) {enabled=false, module=InitialMeshOrientation001}
   &InitialMeshOrientation(Orientation Axis=<0 0 1>, Orientation Vector=<0 0 1>, Rotation=<0 0.25 0>, Use Orientation Vector=true, Use Rotation Vector=true)
   &SkeletalMeshLocation(Bone Sampling Mode=::NewEnumerator4, MaskSpawnGroup=false, Skeletal Mesh=$User.SkeletalMesh)
  ]
  particleUpdateScript [
   &ParticleState()
   &ScaleSpriteSize(Uniform Curve Sprite Scale=DataInterface() {propertyValues {Curve {keys [{interpMode=RCIM_Cubic, leaveTangent=4.353, tangentMode=RCTM_Break, time=0, value=0}, <at 0.2 0.6>, {tangentMode=RCTM_User, time=1, value=1}]}}})
   &ScaleColor(
    Scale Alpha=&FloatFromCurve(FloatCurve=DataInterface() {propertyValues {Curve {keys [
         {interpMode=RCIM_Cubic, leaveTangent=-0.2966, tangentMode=RCTM_Break, time=0, value=1}
         {arriveTangent=-1.4096, interpMode=RCIM_Cubic, tangentMode=RCTM_Break, time=1, value=0}
        ]}}})
    ScaleA=true
    ScaleRGB=true
   )
   &DynamicMaterialParameters(
    Index 0 Param 1=&FloatFromCurve(FloatCurve=DataInterface() {propertyValues {Curve {keys [{arriveTangent=3.726, interpMode=RCIM_Cubic, leaveTangent=3.726, tangentMode=RCTM_User, time=0, value=-0.3}, <at 1 1>]}}})
    Index 0 Param 2=&FloatFromCurve(FloatCurve=DataInterface() {propertyValues {Curve {keys [{arriveTangent=-0.5818, interpMode=RCIM_Cubic, leaveTangent=-0.5818, tangentMode=RCTM_User, time=0, value=1}, <at 1 0.8>]}}})
    Index 0 Param 3=&FloatFromCurve(FloatCurve=DataInterface() {propertyValues {Curve {keys [<at 0 0>,<at 0.15 1>,<at 1 1.2>]}}})
    Module.Param0WriteEnabled=true
    Module.Param1WriteEnabled=true
    Module.Param2WriteEnabled=true
    Module.Param3WriteEnabled=false
   )
  ]
  rendererClasses [&NiagaraMeshRendererProperties]
  renderers [{rendererClass=&NiagaraMeshRendererProperties, rendererIndex=0}]
 }
 StaffGlow {
  emitterUpdateScript [
   &EmitterState(Life Cycle Mode=::NewEnumerator1, Loop Behavior=::NewEnumerator1, Loop Duration=1, UseLoopDelay=false)
   &SpawnBurst_Instantaneous(Use Spawn Probability=false, UseLoopCountLimit=false)
  ]
  particleSpawnScript [
   &InitializeParticle(
    Color=<rgba 2 0.5463 0 0.2>
    Lifetime=5
    Position Offset=<0 8 45>
    Sprite Size=<25 40>
    Sprite Size Mode=::NewEnumerator4
    UsePositionOffset=true
    Write Material Random=true
   )
   &AddVelocity(Velocity=<0 0 1>)
  ]
  particleUpdateScript [&ParticleState(Kill Particles When Lifetime Has Elapsed=false)]
  rendererClasses [&NiagaraSpriteRendererProperties]
  renderers [{rendererClass=&NiagaraSpriteRendererProperties}]
 }
}
summary {
 systemName=NS_Pickaxe_RelayStick_NonReactive_Idle
 emitters [
  {emitterName=OrbGlow, rendererClasses [&NiagaraSpriteRendererProperties, &NiagaraLightRendererProperties]}
  {bEnabled=false, emitterName=OLDMainRing, rendererClasses [&NiagaraMeshRendererProperties]}
  {emitterName=Pulsing, rendererClasses [&NiagaraMeshRendererProperties]}
  {emitterName=StaffGlow, rendererClasses [&NiagaraSpriteRendererProperties]}
  {emitterName=MainRing, rendererClasses [&NiagaraMeshRendererProperties]}
 ]
 userVariables [
  User.bIsFrontEnd {type=NiagaraBool}
  User.bIsFrontEndPreview {type=NiagaraBool}
  User.SkeletalMesh {
   type=NiagaraDataInterfaceSkeletalMesh
   defaultValue {
    struct=NiagaraExt_VariableValue_DataInterface
    value {dataInterface=&self:NiagaraDataInterfaceSkeletalMesh_0}
   }
  }
 ]
}
