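# Material function for the RelayStick pickaxe orb, outputting audio-reactive Emissive and a SpectralThinFilm color.
# A fake cubemap reflection with gradient tint and an animated thin film form the base look, both dimmed at night.
# Panned electric noise in view-derived radial coordinates builds colored lightning, with flowmap-warped fresnel and audio amplitude or spectrum driving intensity.
# Static switches select fixed or dynamic reactivity, and quality switches simplify noise, warp and film on low/medium.
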MaterialFunction /BRCosmetics/Weapons/FORT_Melee/Pickaxe_RelayStick/Materials/MF_RelayStick_Orb
graph {nodes [
  Constant2Vector() {g=0.74, r=5}
  Constant2Vector() {g=1, r=7}
  n1 = NamedRerouteUsage()
  n2 = VectorParameter() {group=RelayStick_Orb, parameterName=ColorLightningSpectrumSensitivity, defaultValue=<rgba 1 2 0 4>}
  n3 = AppendVector(A=@n2.RGB, B=@n2.RGB)
  n4 = MaterialFunctionCall(f4=Multiply(A=@n1, B=Normalize(VectorInput=@n3))) {materialFunction=&AddComponents}
  Divide(A=@n4.f4, B=MaterialFunctionCall(f4=@n2.RGB) {materialFunction=&AddComponents}.f4)
  n5 = MaterialFunctionCall(Float4=@n1) {materialFunction=&BreakOutFloat4Components}
  n6 = Add(A=@n5.R, B=@n5.R)
  Frac(None=@n6)
  n7 = ScalarParameter() {defaultValue=50, group=FakeEnvMap, parameterName=EnvMap_ZenithBrightness}
  n8 = ReflectionVectorWS()
  n9 = RotateAboutAxis(
   NormalizedRotationAxis=Constant3Vector() {constant=<rgba 0 0 1 0>}
   RotationAngle=ScalarParameter() {group=FakeEnvMap, parameterName=ReflectionAngle, sortPriority=-9}
   PivotPoint=Constant3Vector()
   Position=@n8
  )
  n10 = StaticSwitchParameter(True=Normalize(VectorInput=Add(A=@n9, B=@n8)), False=@n8) {group=FakeEnvMap, parameterName=EnableEnvMapRotation, sortPriority=-10}
  n11 = Constant3Vector() {constant=<rgba -1 1 1 0>}
  n12 = Normalize(VectorInput=Add(A=Multiply(A=@n10, B=@n11) {desc=Unmirror}, B=Constant3Vector() {constant=<rgba 0 0 0.5 0>}) {desc=Add perspective distortion to move horizon downward})
  n13 = Saturate(None=DotProduct(A=Constant3Vector() {constant=<rgba 0 0 1 0>}, B=@n12))
  n14 = LinearInterpolate(
   A=ScalarParameter() {defaultValue=0.04, group=FakeEnvMap, parameterName=EnvMap_FresnelBase}
   Alpha=Power(Base=NamedRerouteUsage(), Exp=ScalarParameter() {defaultValue=2.5, group=FakeEnvMap, parameterName=EnvMap_FresnelExp})
  )
  n15 = TextureSampleParameterCube!(UVs=@n12, Apply View MipBias) {
   group=FakeEnvMap
   parameterName=ReflectionEnvMap
   samplerType=SAMPLERTYPE_LinearColor
   texture=/BRCosmetics/Weapons/FORT_Melee/Pickaxe_RelayStick/Textures/TC_RelayStick_EnvMap
  }
  n16 = MaterialFunctionCall!(
   Color=Multiply(A=LinearInterpolate(B=@n7, Alpha=@n13) {constA=1}, B=Multiply(A=@n14, B=@n15.RGB))
   Contrast=ScalarParameter() {defaultValue=0.8, group=FakeEnvMap, parameterName=EnvMap_Contrast}
  ) {materialFunction=/Engine/Functions/Engine_MaterialFunctions01/ImageAdjustment/Contrast_Preserve_Color}
  n17 = ReflectionVectorWS()
  n18 = Constant3Vector() {constant=<rgba 0 0 1 0>}
  n19 = CurveAtlasRowParameter(CurveTime=ConstantBiasScale(None=DotProduct(A=@n17, B=@n18))) {
   atlas=/Game/Weapons/FORT_Melee/Pickaxe_RelayStick/Textures/CA_RelayStick
   curve=/Game/Weapons/FORT_Melee/Pickaxe_RelayStick/Textures/C_RelayStick_EnvMapTint
   parameterName=TintGradient
  }
  n20 = Multiply(A=@n16.Result, B=@n19)
  n21 = ScalarParameter() {defaultValue=0.852, group=FakeEnvMap, parameterName=EnvMap_TintAmount}
  n22 = Constant() {r=1}
  n23 = MaterialFunctionCall(Value_Morning=@n22, Value_Day=@n22, Value_Evening=@n22, Value_Night=Constant()) {materialFunction=&Scalar_Time_of_Day}
  n24 = Multiply(
   A=Multiply(
    A=LinearInterpolate(A=@n16.Result, B=@n20, Alpha=@n21)
    B=ScalarParameter() {defaultValue=0.3, group=FakeEnvMap, parameterName=EnvMap_Brightness}
   )
   B=LinearInterpolate(A=ScalarParameter() {defaultValue=0.3, group=06_Thin Film, parameterName=EnvMap_NightBrightness}, Alpha=@n23.Result)
  )
  n25 = DotProduct(A=PixelNormalWS(), B=SkyAtmosphereLightDirection())
  n26 = Power(Base=Saturate(None=ConstantBiasScale(None=@n25))) {constExponent=4}
  n27 = NamedRerouteUsage()
  n28 = Power(Base=NamedRerouteUsage(), Exp=ScalarParameter() {defaultValue=3, group=06_Thin Film, parameterName=CustomThinFilmExponent})
  n29 = LinearInterpolate(
   A=ScalarParameter() {defaultValue=1, group=06_Thin Film, parameterName=CustomThinFilmIntensity, sortPriority=2}
   B=ScalarParameter() {defaultValue=2, group=06_Thin Film, parameterName=CustomThinFilmIntensity_Loud, sortPriority=2}
   Alpha=Multiply(A=Power(Base=@n27) {constExponent=3}, B=@n28)
  )
  n30 = SmoothStep(Value=Power(Base=@n28, Exp=ScalarParameter() {defaultValue=0.3, group=06_Thin Film, parameterName=CustomThinFilmIntensityExponent, sortPriority=3}))
  n31 = NamedRerouteUsage()
  n32 = ScalarParameter() {defaultValue=0.5, group=06_Thin Film, parameterName=CustomThinFilmWarp, sortPriority=4}
  n33 = Add(
   A=Multiply(A=ScalarParameter() {defaultValue=-0.1, group=06_Thin Film, parameterName=CustomThinFilmPanSpeed}, B=Time())
   B=Multiply(A=ScalarParameter() {defaultValue=0.5, group=06_Thin Film, parameterName=CustomThinFilmScale, sortPriority=5}, B=Power(Base=@n31, Exp=@n32))
  )
  n34 = ScalarParameter() {defaultValue=0.3, group=06_Thin Film, parameterName=CustomThinFilmOffset}
  n35 = ComponentMask(None=NamedRerouteUsage()) {g=true}
  n36 = Multiply(A=Time(), B=ScalarParameter() {defaultValue=0.1, parameterName=CustomThinFilmCircularMaskRotateSpeed})
  n37 = TextureCoordinate()
  n38 = TextureSample!(
   UVs=AppendVector(A=Add(A=@n33, B=@n34), B=Subtract(A=@n35, B=@n36))
   Tex
   "DDX(UVs)"=DDX(Value=@n37)
   "DDY(UVs)"=DDY(Value=@n37)
   Apply View MipBias
  ) {
   mipValueMode=TMVM_Derivative
   samplerType=SAMPLERTYPE_LinearColor
   texture=/Game/Accessories/FORT_Backpacks/Backpack_M_MED_RelayStick/Textures/T_RelayStick_Pattern_A
  }
  n39 = Multiply(A=Time(), B=ScalarParameter() {defaultValue=0.1, parameterName=CustomThinFilmCircularPanSpeed})
  n40 = CurveAtlasRowParameter(CurveTime=Frac(None=Add(A=@n35, B=@n39))) {
   atlas=&CA_RelayStick
   curve=/Game/Weapons/FORT_Melee/Pickaxe_RelayStick/Textures/C_RelayStick_CircularFresnel
   defaultValue=1
   parameterName=CustomThinFilmRadialGradient
  }
  n41 = QualitySwitch(Default=Multiply(A=Add(A=@n38.RGB, B=@n40)) {constB=0.5}, Low=@n40, Medium=@n40)
  n42 = Multiply(
   A=Multiply(A=Multiply(A=@n29, B=@n30), B=@n41)
   B=LinearInterpolate(A=ScalarParameter() {defaultValue=0.3, group=06_Thin Film, parameterName=CustomThinFilm_NightBrightness}, Alpha=@n23.Result)
  )
  n43 = Add(A=@n24, B=Multiply(A=LinearInterpolate(Alpha=@n26) {constA=0.5, constB=4}, B=@n42))
  n44 = LinearInterpolate(
   A=ScalarParameter() {defaultValue=0.04, parameterName=EnvMap_FresnelBase_Loud}
   B=ScalarParameter() {defaultValue=1.5, parameterName=EnvMap_BrightnessScale_Loud}
   Alpha=Power(Base=NamedRerouteUsage(), Exp=ScalarParameter() {defaultValue=1.5, parameterName=EnvMap_FresnelExp_Loud})
  )
  n45 = VectorParameter() {group=RelayStick_Orb, parameterName=Color C, defaultValue=<rgba 0 0 0 2>}
  n46 = NamedRerouteUsage()
  n47 = VectorParameter() {group=RelayStick_Orb, parameterName=OrbBackgroundTint, defaultValue=<rgba 0.0625 0.6958 2 1>}
  n48 = DotProduct(A=PixelNormalWS(), B=CameraVectorWS())
  n49 = Subtract(A=PixelNormalWS(), B=CameraVectorWS())
  n50 = Normalize(VectorInput=Multiply(
    A=Transform(None=@n49) {transformSourceType=TRANSFORMSOURCE_World, transformType=TRANSFORM_Camera}
    B=Constant3Vector() {constant=<rgba 1 1 0 0>}
   ))
  n51 = Constant3Vector() {constant=<rgba 0 1 0 0>}
  n52 = ComponentMask(None=CrossProduct(A=@n50, B=@n51)) {b=true}
  n53 = Abs(None=@n52)
  n54 = MaterialFunctionCall(In=@n53, Power=Constant() {r=0.75}) {
   desc="Kind of optional, but smoothes out\r\nsome nonlinear distortion"
   materialFunction=&SCurve
  }
  n55 = Multiply(A=QualitySwitch(Default=@n54.Result, Low=@n53, Medium=@n53), B=Sign(None=@n52))
  n56 = ConstantBiasScale(None=@n55) {bias=-3, scale=-0.25}
  n57 = ConstantBiasScale(None=@n55) {scale=0.25}
  n58 = Saturate(None=Sign(None=DotProduct(A=@n50, B=@n51)))
  n59 = NamedRerouteDeclaration(None=AppendVector(A=Power(Base=@n48) {desc=Radial coordinate distortion}, B=LinearInterpolate(A=@n56, B=@n57, Alpha=@n58) {constB=0}) {desc=Output UVs}) {name=RadialCoords, nodeColor=<rgba 0.7578 0 1 1>}
  n60 = AppendVector(A=Constant() {r=0.5}, B=ScalarParameter() {defaultValue=4, group=RelayStick_Orb, parameterName=Flowmap UV Scale, sliderMax=1})
  n61 = Multiply(A=ScalarParameter() {defaultValue=0.1, group=RelayStick_Orb, parameterName=Flowmap Panner Speed}, B=TwoSidedSign())
  n62 = Constant()
  n63 = TextureCoordinate() {uTiling=3, vTiling=3}
  n64 = TextureSample(
   UVs=Panner(Coordinate=Multiply(A=@n59, B=@n60), Speed=AppendVector(A=@n61, B=@n62)) {bFractionalPart=true}
   "DDX(UVs)"=DDX(Value=@n63)
   "DDY(UVs)"=DDY(Value=@n63)
  ) {
   mipValueMode=TMVM_Derivative
   samplerSource=SSM_Wrap_WorldGroupSettings
   samplerType=SAMPLERTYPE_LinearColor
   texture=/Game/Effects/Fort_Effects/Textures/Plasma/T_Turbulence_AdditiveWarp_Linear
  }
  n65 = ConstantBiasScale(None=ComponentMask(None=@n64.RGB) {g=true, r=true}) {bias=-0.5, scale=2}
  n66 = Abs(None=DotProduct(A=PixelNormalWS(), B=CameraVectorWS()))
  n67 = SquareRoot(None=@n66)
  n68 = ScalarParameter() {defaultValue=0.5, group=RelayStick_Orb, parameterName=Fresnel Flowmap Intensity Max}
  n69 = ScalarParameter() {defaultValue=0.1, group=RelayStick_Orb, parameterName=Fresnel Flowmap Intensity Min}
  n70 = AppendVector(A=Multiply(A=Multiply(A=@n65, B=@n67), B=LinearInterpolate(A=@n68, B=@n69, Alpha=@n66)), B=Constant())
  n71 = Normalize(VectorInput=Add(A=Transform(None=@n70) {transformSourceType=TRANSFORMSOURCE_View}, B=PixelNormalWS()))
  n72 = VertexNormalWS()
  n73 = Abs(None=DotProduct(A=QualitySwitch(Default=@n71, Low=@n72, Medium=@n72) {desc=Skipping flowmap on med/low}, B=CameraVectorWS()))
  n74 = OneMinus(None=@n73)
  n75 = VectorParameter() {group=RelayStick_Orb, parameterName="OrbRadialCoordScales (UVUV)", defaultValue=<rgba 0.1366 2 0.174 1>}
  n76 = AppendVector(A=@n75.R, B=@n75.R)
  n77 = Add(
   A=Time()
   B=Multiply(A=NamedRerouteUsage(), B=ScalarParameter() {defaultValue=0.1, group=RelayStick_Orb, parameterName=Color Lightning Audio Jitter})
  )
  n78 = VectorParameter() {group=RelayStick_Orb, parameterName="OrbRadialCoordPanSpeeds (UVUV)", defaultValue=<rgba 0.0333 0.0333 -0.0267 -0.0233>}
  n79 = TextureCoordinate() {uTiling=3, vTiling=3}
  n80 = DDX(Value=@n79)
  n81 = DDY(Value=@n79)
  n82 = TextureSample(
   UVs=Panner(Coordinate=Multiply(A=@n59, B=@n76) {constB=3}, Time=@n77, Speed=AppendVector(A=@n78.R, B=@n78.R)) {speedX=-1.1, speedY=-0.067}
   "DDX(UVs)"=@n80
   "DDY(UVs)"=@n81
  ) {
   mipValueMode=TMVM_Derivative
   samplerSource=SSM_Wrap_WorldGroupSettings
   samplerType=SAMPLERTYPE_LinearColor
   texture=/Game/Characters/Player/Male/Medium/Bodies/M_MED_SRDrift/Textures/T_ElectricNoise
  }
  n83 = AppendVector(A=@n75.B, B=@n75.B)
  n84 = TextureSample(
   UVs=Panner(Coordinate=Multiply(A=@n59, B=@n83), Time=@n77, Speed=AppendVector(A=@n78.B, B=@n78.B)) {speedX=-0.33, speedY=-0.053}
   "DDX(UVs)"=@n80
   "DDY(UVs)"=@n81
  ) {
   mipValueMode=TMVM_Derivative
   samplerSource=SSM_Wrap_WorldGroupSettings
   samplerType=SAMPLERTYPE_LinearColor
   texture=&T_ElectricNoise
  }
  n85 = MaterialFunctionCall!(Blend=NamedRerouteUsage(), Base=QualitySwitch(Default=Add(A=@n82.R, B=@n84.R), Low=@n82.R, Medium=@n82.R)) {materialFunction=/Engine/Functions/Engine_MaterialFunctions03/Blends/Blend_Screen}
  n86 = ScalarParameter() {defaultValue=3, group=RelayStick_Orb, parameterName=Blended Texture Exponent}
  n87 = VectorParameter() {group=RelayStick_Orb, parameterName=Color Lightning Fresnel, defaultValue=<rgba 1.3 0 3 0.1>}
  n88 = Power(Base=@n73, Exp=ScalarParameter() {defaultValue=3, group=RelayStick_Orb, parameterName=Lightning Edge Fade Exponent}) {constExponent=3}
  n89 = ScalarParameter() {defaultValue=3, group=RelayStick_Orb, parameterName=Color Lightning Sensitivity}
  n90 = Multiply(
   A=LinearInterpolate(
    A=VectorParameter() {group=RelayStick_Orb, parameterName=Color Lightning Falloff, defaultValue=<rgba 0.015 0 0.9 3>}.RGB
    B=VectorParameter() {group=RelayStick_Orb, parameterName=Color Lightning, defaultValue=<rgba 0.07 0.845 1 0>}.RGB
    Alpha=Multiply(A=@n85.Result, B=@n88)
   )
   B=LinearInterpolate(
    A=ScalarParameter() {defaultValue=1, group=RelayStick_Orb, parameterName=Color Lightning Intensity}
    B=ScalarParameter() {defaultValue=2, group=RelayStick_Orb, parameterName=Color Lightning Intensity Loud}
    Alpha=Power(Base=@n4.f4, Exp=@n89)
   ) {constA=0.2, constB=2}
  )
  n91 = Power(Base=@n73, Exp=@n87.A)
  n92 = Add(
   A=LinearInterpolate(A=@n45.RGB, B=Multiply(A=@n46, B=@n47.RGB), Alpha=Power(Base=@n74, Exp=@n45.A))
   B=Multiply(A=Power(Base=@n85.Result, Exp=@n86), B=LinearInterpolate(A=@n87.RGB, B=@n90, Alpha=@n91))
  ) {constB=0}
  n93 = OneMinus(None=Power(Base=NamedRerouteUsage()) {constExponent=0.1})
  n94 = ScalarParameter() {defaultValue=1, group=RelayStick_Orb, parameterName=FX Strength}
  n95 = Add(
   A=LinearInterpolate(
    A=Multiply(A=@n44, B=@n24)
    B=MaterialFunctionCall(Emissive=@n92) {materialFunction=&MF_FX_EmissiveAdjustments}.Result
    Alpha=Multiply(A=@n93, B=@n94)
   )
   B=@n42
  )
  n96 = NamedRerouteUsage()
  n97 = LinearInterpolate(A=@n43, B=@n95, Alpha=Saturate(None=Multiply(A=@n96) {constB=100}))
  FunctionOutput(None=StaticSwitchParameter(True=StaticSwitchParameter(True=@n97, False=@n95) {parameterName=EnableDynamicOrbReactivity}, False=@n43) {group=RelayStick_Orb, parameterName=EnableOrbReactivity}) {outputName=Emissive}
  FunctionOutput(None=NamedRerouteDeclaration(None=@n41) {name=SpectralThinFilm, nodeColor=<rgba 0.8516 1 0 1>}) {outputName=SpectralThinFilm, sortPriority=10}
  n98 = ScalarParameter() {defaultValue=0.75, group=RelayStick_Orb, parameterName=Center Glow Size}
  n99 = ScalarParameter() {defaultValue=2, group=RelayStick_Orb, parameterName=Center Glow Size_Loud}
  n100 = Power(Base=NamedRerouteUsage()) {constExponent=3}
  n101 = Saturate(None=Divide(A=@n74, B=LinearInterpolate(A=@n98, B=@n99, Alpha=@n100)))
  NamedRerouteDeclaration(None=Power(Base=OneMinus(None=@n101), Exp=ScalarParameter() {defaultValue=22, group=RelayStick_Orb, parameterName=Center Glow Exponent})) {name=Center Glow Mask, nodeColor=<rgba 1 0.8203 0 1>}
  n102 = MaterialFunctionCall() {materialFunction=/Game/Packages/Fortress_SharedMaterials/Base_Material_Functions/BaseCharacterShaderFunctions/Audio/MF_AudioReactivity}
  NamedRerouteDeclaration(None=FunctionInput(Preview=@n102.Amplitude) {bUsePreviewValueAsDefault=true, inputName=AudioAmplitude, inputType=FunctionInput_Scalar}) {name=Amplitude, nodeColor=<rgba 0 0.0625 1 1>}
  NamedRerouteDeclaration(None=FunctionInput(Preview=@n102.Spectrum) {bUsePreviewValueAsDefault=true, inputName=AudioSpectrum, inputType=FunctionInput_Vector4}) {name=Spectrum, nodeColor=<rgba 0.8281 0 1 1>}
  NamedRerouteDeclaration(None=FunctionInput(Preview=Fresnel() {baseReflectFraction=0, exponent=1}) {bUsePreviewValueAsDefault=true, inputName=Fresnel, inputType=FunctionInput_Scalar}) {name=Fresnel, nodeColor=<rgba 0.4531 0 1 1>}
  NamedRerouteUsage()
  SmoothStep(Value=@n6)
 ]}
