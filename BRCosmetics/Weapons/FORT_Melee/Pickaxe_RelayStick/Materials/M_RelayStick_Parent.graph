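# Parent material for the RelayStick pickaxe cosmetic, layering audio-reactive effects between the 2023 character parent PreFX and PostFX functions.
# UseCustomReflection blends in a masked orb with audio-driven emissive and fake GGX sky-light specular.
# ReactiveVariant adds five-color panning noise over a sphere-masked mask-channel value, with glow and pan wobble driven by rerouted audio signals.
# Within that variant, Enable Accent Tint applies FX2-masked accent tint, seam color and emissive.
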Material /BRCosmetics/Weapons/FORT_Melee/Pickaxe_RelayStick/Materials/M_RelayStick_Parent
graph {
 properties {
  bHasPixelAnimation=true
  bUseMaterialAttributes=true
  bUsedWithClothing=true
  bUsedWithMeshDeformer=true
  bUsedWithMorphTargets=true
  bUsedWithNanite=true
  bUsedWithNiagaraMeshParticles=true
  bUsedWithSkeletalMesh=true
  pixelDepthOffsetMode=PDOM_Legacy
  shadingModel=MSM_FromMaterialExpression
  usedShadingModels=Default Lit | Subsurface | Clear Coat | Subsurface Profile | Cloth
 }
 nodes [
  n1 = NamedRerouteUsage()
  n2 = VectorParameter() {parameterName=Reactive_EmissiveSpectrumSensitivity, defaultValue=<rgba 1 2 0 2>}
  n3 = Normalize(VectorInput=AppendVector(A=@n2.RGB, B=@n2.RGB))
  DotProduct(A=@n1, B=@n3)
  n4 = MaterialFunctionCall() {materialFunction=&MF_CharacterParent_2023_PreFX}
  n5 = GetMaterialAttributes(None=@n4.Result) {attributeGetTypes [#1, #2]}
  n6 = MaterialFunctionCall!(
   AudioSpectrum=NamedRerouteUsage()
   Fresnel=NamedRerouteDeclaration(None=@n4.Fresnel) {name=Fresnel, nodeColor=<rgba 0.0781 0 1 1>}
   AudioAmplitude=Power(Base=NamedRerouteUsage(), Exp=ScalarParameter() {defaultValue=1, group=RelayStick, parameterName=Orb_AudioSensitivity})
  ) {materialFunction=&/MF_RelayStick_Orb}
  n7 = PixelNormalWS()
  n8 = SkyAtmosphereLightDirection()
  n9 = LinearInterpolate(A=ReflectionVectorWS(), B=Multiply(A=@n8) {constB=-1})
  n10 = ScalarParameter() {defaultValue=0.1, group=RelayStick, parameterName=OrbRoughness, sliderMax=1}
  n11 = ConstantBiasScale(None=DotProduct(A=PixelNormalWS(), B=@n8))
  n12 = Multiply(
   A=Multiply(
    A=MaterialFunctionCall!(Normal=@n7, Light Vector=@n9, Roughness=@n10) {materialFunction=/Engine/Functions/Engine_MaterialFunctions01/Lighting/GGXSpecular}.Result
    B=Power(Base=@n11) {constExponent=8}
   )
   B=ScalarParameter() {defaultValue=1, group=RelayStick, parameterName=OrbFakeSpecularIntensity}
  )
  n13 = SetMaterialAttributes!(
   MaterialAttributes=@n5.MaterialAttributes
   Emissive Color=Add(A=@n5.Emissive Color, B=Add(A=@n6.Emissive, B=@n12))
   Base Color=Multiply(A=@n5.Base Color, B=ScalarParameter() {defaultValue=0.3, group=RelayStick, parameterName=OrbBaseColorScale})
   Roughness=ScalarParameter() {defaultValue=0.1, group=RelayStick, parameterName=OrbRoughness, sliderMax=1}
   Specular=ScalarParameter() {group=RelayStick, parameterName=OrbSpecular, sliderMax=1}
   Metallic=ScalarParameter() {group=RelayStick, parameterName=OrbMetallic, sliderMax=1}
  ) {attributeSetTypes [
    #1
    #2
    #3
    #4
    #5
   ]}
  n14 = StaticSwitchParameter(
   True=BlendMaterialAttributes(
    A=@n5.MaterialAttributes
    B=@n13
    Alpha=ChannelMaskParameter(None=@n4.Tech Art Mask) {group=RelayStick, maskChannel=Blue, parameterName=ReflectionChannel, defaultValue=<rgba 0 0 1 0>}.R
   ) {vertexAttributeBlendType=UseA}
   False=@n4.Result
  ) {group=RelayStick, parameterName=UseCustomReflection}
  n15 = GetMaterialAttributes(None=@n14) {attributeGetTypes [#2, #1]}
  n16 = Multiply(A=@n1, B=@n3)
  n17 = Multiply(
   A=Power(
    Base=MaterialFunctionCall(f4=@n16) {materialFunction=&AddComponents}.f4
    Exp=ScalarParameter() {defaultValue=2, group=RelayStick, parameterName=Reactive_EmissiveSensitivity}
   )
   B=Multiply(A=ScalarParameter() {defaultValue=4, group=RelayStick, parameterName=Reactive_EmissiveStrength}, B=NamedRerouteUsage())
  )
  n18 = LinearInterpolate(
   A=Multiply(A=ScalarParameter() {defaultValue=2, group=RelayStick, parameterName=Reactive_PatternUVScale}, B=TextureCoordinate())
   B=NamedRerouteUsage()
   Alpha=ScalarParameter() {defaultValue=0.3, group=RelayStick, parameterName=Reactive_FresnelCoordsIntensity, sliderMax=1}
  )
  n19 = Power(Base=NamedRerouteUsage(), Exp=ScalarParameter() {defaultValue=6, group=RelayStick, parameterName=Reactive_AmplitudeWobbleExponent})
  n20 = ScalarParameter() {defaultValue=0.5, group=RelayStick, parameterName=Reactive_AmplitudeWobbleAmt}
  n21 = Multiply(
   A=ScalarParameter() {defaultValue=0.1, group=RelayStick, parameterName=Reactive_PanSpeed}
   B=Add(A=Time(), B=Multiply(A=@n19, B=@n20) {constB=0.2})
  )
  n22 = TextureSampleParameter2D(UVs=Panner(Coordinate=@n18, Time=@n21) {speedX=0.474, speedY=-0.99}) {
   group=RelayStick
   parameterName=ReactiveNoiseOverlay
   samplerSource=SSM_Wrap_WorldGroupSettings
   samplerType=SAMPLERTYPE_LinearColor
   texture=/Game/Accessories/FORT_Backpacks/Backpack_M_MED_RelayStick/Textures/T_RelayStick_Pattern_ReactiveMask
  }
  n23 = TextureSampleParameter2D(UVs=Panner(Coordinate=Multiply(A=@n18) {constB=1.3712}, Time=@n21) {speedX=0.69, speedY=0.354}) {
   group=RelayStick
   parameterName=ReactiveNoiseOverlay
   samplerSource=SSM_Wrap_WorldGroupSettings
   samplerType=SAMPLERTYPE_LinearColor
   texture=&T_RelayStick_Pattern_ReactiveMask
  }
  n24 = LinearInterpolate(
   A=VectorParameter() {group=RelayStick, parameterName=Reactive_ColorA, defaultValue=<rgba 0.3281 0.0068 0.0658 1>}.RGB
   B=VectorParameter() {group=RelayStick, parameterName=Reactive_ColorB, defaultValue=<rgba 0 0.0432 0.2969 1>}.RGB
   Alpha=Power(Base=Multiply(A=@n22.R, B=@n23.R), Exp=ScalarParameter() {defaultValue=1, group=RelayStick, parameterName=Reactive_ColorAB_Exponent})
  )
  n25 = VectorParameter() {group=RelayStick, parameterName=Reactive_ColorC, defaultValue=<rgba 0 4.6667 7 1>}
  n26 = TextureSampleParameter2D(UVs=Panner(Coordinate=Multiply(A=@n18) {constB=0.7235}, Time=@n21) {speedX=-1.125, speedY=0.038}) {
   group=RelayStick
   parameterName=ReactiveNoiseOverlay
   samplerSource=SSM_Wrap_WorldGroupSettings
   samplerType=SAMPLERTYPE_LinearColor
   texture=&T_RelayStick_Pattern_ReactiveMask
  }
  n27 = Add(A=@n22.R, B=@n26.R)
  n28 = Multiply(
   A=Power(Base=Divide(A=@n27), Exp=ScalarParameter() {defaultValue=6, group=RelayStick, parameterName=Reactive_ColorC_Exponent})
   B=ScalarParameter() {defaultValue=1, group=RelayStick, parameterName=Reactive_ColorC_Amt}
  )
  n29 = Power(Base=Divide(A=Add(A=@n23.R, B=@n26.R)), Exp=ScalarParameter() {defaultValue=3, group=RelayStick, parameterName=Reactive_ColorD_Exponent})
  n30 = ScalarParameter() {defaultValue=1, group=RelayStick, parameterName=Reactive_ColorD_Amt}
  n31 = OneMinus(None=@n26.R)
  n32 = ScalarParameter() {defaultValue=5, group=RelayStick, parameterName=Reactive_ColorE_Exponent}
  n33 = LinearInterpolate(
   A=LinearInterpolate(
    A=LinearInterpolate(A=@n24, B=@n25.RGB, Alpha=@n28)
    B=VectorParameter() {group=RelayStick, parameterName=Reactive_ColorD, defaultValue=<rgba 0 0.3867 5 1>}.RGB
    Alpha=Multiply(A=@n29, B=@n30)
   )
   B=VectorParameter() {group=RelayStick, parameterName=Reactive_ColorE, defaultValue=<rgba 0 0.3867 5 1>}.RGB
   Alpha=Multiply(A=Power(Base=@n31, Exp=@n32), B=ScalarParameter() {defaultValue=1, group=RelayStick, parameterName=Reactive_ColorE_Amt})
  )
  n34 = NamedRerouteUsage()
  n35 = ScalarParameter() {defaultValue=2, group=RelayStick, parameterName=Reactive_EmissiveExponent}
  n36 = ChannelMaskParameter(None=@n4.Tech Art Mask) {group=RelayStick, maskChannel=Blue, parameterName=AudioReactivityMaskChannel, defaultValue=<rgba 0 0 1 0>}
  n37 = SphereMask(
   A=@n36.R
   B=ScalarParameter() {defaultValue=0.5, group=RelayStick, parameterName=AudioReactivityMask_FXMaskValue}
   Radius=ScalarParameter() {defaultValue=0.1, group=RelayStick, parameterName=AudioReactivityMask_FXMaskValueRadius}
   Hardness=ScalarParameter() {defaultValue=0.5, group=RelayStick, parameterName=AudioReactivityMask_FXMaskValueHardness}
  ) {attenuationRadius=0.1, hardnessPercent=50}
  n38 = Multiply(A=Multiply(A=Multiply(A=@n33, B=@n17), B=Power(Base=@n34, Exp=@n35)), B=@n37)
  n39 = BlendMaterialAttributes(
   A=@n14
   B=SetMaterialAttributes!(MaterialAttributes=@n15.MaterialAttributes, Base Color=@n33, Emissive Color=Add(A=@n15.Emissive Color, B=@n38)) {attributeSetTypes [#2, #1]}
   Alpha=@n37
  ) {vertexAttributeBlendType=UseA}
  n40 = GetMaterialAttributes(None=@n39) {attributeGetTypes [#2, #1]}
  n41 = Multiply(
   A=Multiply(A=Desaturation(None=@n40.Base Color), B=VectorParameter() {group=RelayStick, parameterName=AccentTint, defaultValue=<rgba 0 0.6697 1 1>}.RGB)
   B=ScalarParameter() {defaultValue=4, group=RelayStick, parameterName=AccentBrightness}
  )
  n42 = TextureSampleParameter2D() {
   group=RelayStick
   parameterName=FX2_TintSeamsMasks
   samplerSource=SSM_Wrap_WorldGroupSettings
   samplerType=SAMPLERTYPE_Masks
   texture=/BRCosmetics/Characters/Player/Male/Medium/Bodies/M_MED_RelayStick/Textures/T_RelayStick_Body_FX2
  }
  n43 = Power(
   Base=ChannelMaskParameter(None=@n42.RGBA) {group=RelayStick, maskChannel=Blue, parameterName=TintChannel, defaultValue=<rgba 0 0 1 0>}.R
   Exp=ScalarParameter() {defaultValue=1, group=RelayStick, parameterName=AccentTintCurve}
  )
  n44 = LinearInterpolate(
   A=VectorParameter() {group=RelayStick, parameterName=SeamsTint, defaultValue=<rgba 0 1 0.7355 1>}.RGB
   B=VectorParameter() {group=RelayStick, parameterName=SeamsTint2, defaultValue=<rgba 0 1 0.4914 1>}.RGB
   Alpha=Power(Base=NamedRerouteUsage(), Exp=ScalarParameter() {defaultValue=1, parameterName=SeamsTintFresnelExp})
  )
  n45 = ChannelMaskParameter(None=@n42.RGBA) {group=RelayStick, parameterName=SeamsChannel}
  n46 = Multiply(A=@n44, B=ScalarParameter() {defaultValue=5, parameterName=EmissiveAccentBrightness})
  n47 = ChannelMaskParameter(None=@n42.RGBA) {group=RelayStick, maskChannel=Green, parameterName=EmissiveAccentChannel, defaultValue=<rgba 0 1 0 0>}
  n48 = SetMaterialAttributes(
   MaterialAttributes=@n40.MaterialAttributes
   Base Color=Multiply(A=LinearInterpolate(A=@n40.Base Color, B=@n41, Alpha=@n43), B=LinearInterpolate(B=@n44, Alpha=@n45.R) {constA=1})
   Emissive Color=Add(A=Multiply(A=@n46, B=@n47.R), B=@n40.Emissive Color)
  ) {attributeSetTypes [#2, #1]}
  MaterialFunctionCall(
   Material=StaticSwitchParameter(True=StaticSwitchParameter(True=@n48, False=@n39) {group=RelayStick, parameterName=Enable Accent Tint}, False=@n14) {group=RelayStick, parameterName=ReactiveVariant}
   M=@n4.M
   Pre Skinned Local Position=@n4.M
   Distance Blend=@n4.M
   Fresnel=@n4.M
  ) {materialFunction=&MF_CharacterParent_2023_PostFX}
  n49 = MaterialFunctionCall() {materialFunction=/Game/Packages/Fortress_SharedMaterials/Base_Material_Functions/BaseCharacterShaderFunctions/Audio/MF_AudioReactivity}
  n50 = ScalarParameter() {defaultValue=1, group=RelayStick, parameterName=EnableAudioReactivity, sliderMax=1}
  NamedRerouteDeclaration(None=Multiply(A=@n49.Amplitude, B=@n50)) {name=Amplitude, nodeColor=<rgba 1 0 0.2578 1>}
  NamedRerouteDeclaration(None=Multiply(A=@n49.Spectrum, B=@n50)) {name=Spectrum, nodeColor=<rgba 1 0 0.8203 1>}
 ]
}
