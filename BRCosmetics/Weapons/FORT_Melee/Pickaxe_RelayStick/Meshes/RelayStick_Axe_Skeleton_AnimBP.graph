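# Animation Blueprint for the RelayStick pickaxe skeleton that makes its dyn_ball bone bob and react springily to motion.
# The update event builds a vector from real time, using sine waves of different frequencies and amplitudes on X and Z.
# A Modify Bone node adds this as a bone-space translation, feeding an Anim Dynamics linear spring with gravity, damped component motion, tight linear limits and zero angular limits.
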AnimBlueprint /BRCosmetics/Weapons/FORT_Melee/Pickaxe_RelayStick/Meshes/RelayStick_Axe_Skeleton_AnimBP
target_skeleton=&/RelayStick_Axe_Skeleton
graphs [
 AnimGraph {nodes [
   n1 = "Transform (Modify) Bone - Bone: dyn_ball"(
    ComponentPose=Local To Component(LocalPose=Input Pose())
    Alpha=1
    Translation=Get sinFunction() -> (sinFunction:Vector)
    Rotation="0,0,0"
    Scale="1,1,1"
   ) {translationMode=BMM_Additive, translationSpace=BCS_BoneSpace, alphaBoolBlend {blendOption=Linear}, boneToModify {boneName=dyn_ball}}
   n2 = Anim Dynamics - Bone: dyn_ball(ComponentPose=@n1, Alpha=1, ExternalForce="0,0,0") {
    bLinearSpring=true
    bOverrideLinearDamping=true
    bUseGravityOverride=true
    linearDampingOverride=1
    linearSpringConstant=200
    simulationSpace=RootRelative
    alphaBoolBlend {blendOption=Linear}
    alphaScaleBiasClamp {scale=0.2}
    boundBone {boneName=dyn_ball}
    componentLinearAccScale=<0.3 0.3 0.3>
    componentLinearVelScale=<0.3 0.3 0.3>
    gravityOverride {z=-980}
    physicsBodyDefinitions [{
      boundBone {boneName=dyn_ball}
      boxExtents=<300 300 300>
      constraintSetup {angularLimitsMax=<0 0 0>, angularLimitsMin=<0 0 0>, angularTarget=<1 0 0>, linearAxesMax=<5 5 5>, linearAxesMin=<-5 -5 -5>}
      localJointOffset=<0 0 -1>
     }]
   }
   Output Pose(Result=Component To Local(ComponentPose=@n2))
  ]}
 EventGraph {nodes [
   Event BlueprintUpdateAnimation()
   n1 = GetRealTimeSeconds()
   n2 = "Sin (Radians)"(A=float * float(A=@n1, B=3))
   n3 = float * float(A=@n1, B=2.1)
   > Set sinFunction(sinFunction:Vector=MakeVector(X=float / float(A=@n2, B=2) {is=title K2Node_CallFunction}, Y=0, Z="Sin (Radians)"(A=@n3))) -> (Output_Get:Vector)
   TryGetPawnOwner()
  ]}
]
variables [sinFunction {type=Vector}]
